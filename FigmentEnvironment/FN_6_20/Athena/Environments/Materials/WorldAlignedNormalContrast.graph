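# Material function that triplanar-projects a normal map in world space, with adjustable contrast on projection transitions.
# World position over TextureSize gives the UVs, and CheapContrast on absolute world-normal components sets the blend weights.
# Static switches select Engine's WorldAlignedNormals_HighQuality or a cheaper three-sample blend, though the XYZ switch is not wired to that quality input.
# XYZ, XY, Z and flat-top outputs return tangent-space normals, or world-space via WorldSpace.
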MaterialFunction /FigmentEnvironment/FN_6_20/Athena/Environments/Materials/WorldAlignedNormalContrast
graph {
 properties {description=Tiles a texture in worldspace, libraryCategoriesText [Misc,Texturing]}
 nodes [
  Constant()
  n1 = FunctionInput() {bUsePreviewValueAsDefault=true, description=Size of texture in worldspace units, inputName=TextureSize, previewValue=<64 64 64 6>}
  n2 = Multiply(A=Abs(None=@n1)) {constA=-1, constB=-1}
  n3 = FunctionInput(Preview=WorldPosition() {worldPositionShaderOffset=WPT_ExcludeAllShaderOffsets}.XYZ) {bUsePreviewValueAsDefault=true, inputName=WorldPosition, sortPriority=100}
  n4 = FunctionInput(Preview=TextureObject() {samplerType=SAMPLERTYPE_LinearColor, texture=/Engine/EngineMaterials/DefaultNormal_Uncompressed}) {inputName=TextureObject, inputType=FunctionInput_Texture2D}
  n5 = Transform(None=FunctionInput() {bUsePreviewValueAsDefault=true, inputName=Normal, previewValue=<0 0 1 1>})
  n6 = Multiply(A=@n5, B=Constant() {r=1})
  n7 = ComponentMask(None=@n6) {b=true}
  n8 = FunctionInput() {
   bUsePreviewValueAsDefault=true
   inputName=ProjectionTransitionContrast
   inputType=FunctionInput_Scalar
   sortPriority=10
   previewValue=<1 0 0 1>
  }
  n9 = MaterialFunctionCall(In=Abs(None=@n7), Contrast=@n8) {materialFunction=&CheapContrast}
  n10 = ComponentMask(None=@n6) {r=true}
  n11 = MaterialFunctionCall(In=Abs(None=@n10), Contrast=@n8) {materialFunction=&CheapContrast}
  n12 = MaterialFunctionCall!(
   Negative World Position Divisor=@n2
   World Position=@n3
   Texture=@n4
   Z lerp Alpha=OneMinus(None=@n9.Result)
   WorldNormal
   X lerp alpha=OneMinus(None=@n11.Result)
  ) {materialFunction=/Engine/Functions/Engine_MaterialFunctions02/WorldAlignedNormals_HighQuality}
  n13 = Divide(A=@n3, B=@n2)
  n14 = TextureSample(UVs=ComponentMask(None=@n13) {b=true, r=true}, Tex=@n4) {samplerType=SAMPLERTYPE_Normal}
  n15 = Multiply(A=@n14.R) {constB=-1}
  n16 = Multiply(A=@n14.B, B=ComponentMask(None=@n6) {g=true})
  n17 = TextureSample(UVs=ComponentMask(None=@n13) {b=true, g=true}, Tex=@n4) {samplerType=SAMPLERTYPE_Normal}
  n18 = Multiply(A=@n10, B=@n17.B)
  n19 = Multiply(A=@n17.R) {constB=-1}
  n20 = LinearInterpolate(
   A=AppendVector(A=AppendVector(A=@n15, B=@n16), B=Multiply(A=@n14.G) {constB=-1})
   B=AppendVector(A=AppendVector(A=@n18, B=@n19), B=Multiply(A=@n17.G) {constB=-1})
   Alpha=@n11.Result
  )
  n21 = TextureSample(UVs=ComponentMask(None=@n13) {g=true, r=true}, Tex=@n4) {samplerType=SAMPLERTYPE_Normal}
  n22 = Constant3Vector() {constant=<rgba -1 -1 1 0>}
  n23 = AppendVector(A=ComponentMask(None=Multiply(A=@n21.RGB, B=@n22)) {g=true, r=true}, B=Multiply(A=@n21.B, B=@n7))
  n24 = StaticSwitch(True=@n12.World Space Normals XYZ, False=Normalize(VectorInput=LinearInterpolate(A=@n20, B=@n23, Alpha=@n9.Result)))
  n25 = FunctionInput(Preview=StaticBool()) {bUsePreviewValueAsDefault=true, description=This switch defaults to false., inputName=WorldSpace, inputType=FunctionInput_StaticBool}
  FunctionOutput(None=StaticSwitch(True=@n24, False=Transform(None=@n24) {transformSourceType=TRANSFORMSOURCE_World, transformType=TRANSFORM_Tangent}, Value=@n25)) {outputName=XYZ Texture, sortPriority=1}
  n26 = FunctionInput(Preview=StaticBool() {desc=Use High quality normals, value=true}) {
   bUsePreviewValueAsDefault=true
   description="Default is true. Adds ~ 20 instructions."
   inputName=Use High Quality Normals
   inputType=FunctionInput_StaticBool
   sortPriority=150
  }
  n27 = StaticSwitch(True=@n12.XY Projection World Space Normals, False=Normalize(VectorInput=@n20), Value=@n26)
  FunctionOutput(None=StaticSwitch(True=@n27, False=Transform(None=@n27) {transformSourceType=TRANSFORMSOURCE_World, transformType=TRANSFORM_Tangent}, Value=@n25)) {outputName=XY Texture, sortPriority=1}
  n28 = StaticSwitch(True=@n12.Z Projection World Space Normals, False=Normalize(VectorInput=@n23), Value=@n26)
  FunctionOutput(None=StaticSwitch(True=@n28, False=Transform(None=@n28) {transformSourceType=TRANSFORMSOURCE_World, transformType=TRANSFORM_Tangent}, Value=@n25)) {outputName=Z Texture, sortPriority=1}
  n29 = StaticSwitch(
   True=@n12.XY Flat Top Projection World Space Normals
   False=Normalize(VectorInput=LinearInterpolate(A=@n20, B=@n5, Alpha=@n9.Result))
   Value=@n26
  )
  FunctionOutput(None=StaticSwitch(True=@n29, False=Transform(None=@n29) {transformSourceType=TRANSFORMSOURCE_World, transformType=TRANSFORM_Tangent}, Value=@n25)) {outputName=XYZFlatTop, sortPriority=1}
 ]
}
